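# Sound cue for collecting the Bacon health ingredient, layering item foley with generic loot and reward stingers.
# A mixer plays the find_toilet_bacon wave and a special loot effect at once, plus a generic UI reward sound after a fixed delay, each at its own level.
# It uses prop-collect attenuation and the player-interact sound class, with a concurrency override that caps overlapping copies by stopping the oldest.
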SoundCue /Game/Sounds/Foley_Loot/Ingredients/Health/Bacon_Collect_Cue
root=Mixer_4
nodes [
 {id=Delay_4, children [WavePlayer_11], properties {delayMax=1.2, delayMin=1.2}}
 {id=Mixer_4, children [WavePlayer_6, Delay_4, WavePlayer_12], properties {inputVolume [0.7,0.8,1]}}
 {id=WavePlayer_11, properties {soundWaveAssetPtr=/Game/Sounds/Fort_UI_Sounds/Gameplay/Reward_Generic}}
 {id=WavePlayer_12, properties {soundWaveAssetPtr=/Game/Sounds/Foley_Loot/Ingredients/Generic/find_loot_special_effect}}
 {id=WavePlayer_6, properties {soundWaveAssetPtr=&/find_toilet_bacon}}
]
properties {
 attenuationSettings=/Game/Sounds/Attenuation/Fort_Prop_Collect_Att
 bOverrideConcurrency=true
 maxDistance=2097152
 soundClassObject=/Game/Sounds/FortSoundClassesAndModes/Fort_08_Player_Interacts
 volumeMultiplier=0.7
 concurrencyOverrides {maxCount=2, resolutionRule=StopOldest}
}
