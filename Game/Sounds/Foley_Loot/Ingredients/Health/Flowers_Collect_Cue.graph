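# Sound cue that plays the pickup sound for collecting flowers, a health ingredient in the loot foley set.
# A mixer layers a "real" and a "surreal" flower-collect wave at equal volume, so both play together.
# It uses the prop-collect attenuation and the loot-collect sound class at a reduced volume.
# A concurrency override limits it to two instances, stopping the oldest when exceeded.
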SoundCue /Game/Sounds/Foley_Loot/Ingredients/Health/Flowers_Collect_Cue
root=Mixer_11
nodes [
 {id=Mixer_11, children [WavePlayer_25, WavePlayer_24], properties {inputVolume [1,1]}}
 {id=WavePlayer_24, properties {soundWaveAssetPtr=&/Flowers_Collect_Surreal}}
 {id=WavePlayer_25, properties {soundWaveAssetPtr=&/Flowers_Collect_Real}}
]
properties {
 attenuationSettings=/Game/Sounds/Attenuation/Fort_Prop_Collect_Att
 bOverrideConcurrency=true
 maxDistance=2097152
 soundClassObject=&Fort_Loot_Collect
 volumeMultiplier=0.326
 concurrencyOverrides {maxCount=2, resolutionRule=StopOldest}
}
