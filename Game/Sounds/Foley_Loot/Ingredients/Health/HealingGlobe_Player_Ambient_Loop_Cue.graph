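# Sound cue for the looping ambient sound of a health healing globe, named as a player variant, in the player-interaction sound class.
# A short random delay leads into a modulator, with its pitch ceiling raised for variation, over a single looping wave.
# It uses stereo collectible-prop attenuation, and a concurrency override allows two instances, stopping the oldest beyond that.
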SoundCue /Game/Sounds/Foley_Loot/Ingredients/Health/HealingGlobe_Player_Ambient_Loop_Cue
root=Delay_0
nodes [
 {id=Delay_0, children [Modulator_2], properties {delayMax=0.3, delayMin=0.1}}
 {id=Modulator_2, children [WavePlayer_5], properties {pitchMax=1.5}}
 {id=WavePlayer_5, properties {bLooping=true, soundWaveAssetPtr=&/STW_HealthGlobe_Loop_01}}
]
properties {
 attenuationSettings=/Game/Sounds/Attenuation/Fort_Prop_Collect_Stereo_Att
 bOverrideConcurrency=true
 maxDistance=1150
 soundClassObject=/Game/Sounds/FortSoundClassesAndModes/Fort_08_Player_Interacts
 volumeMultiplier=0.6
 concurrencyOverrides {maxCount=2, resolutionRule=StopOldest}
}
