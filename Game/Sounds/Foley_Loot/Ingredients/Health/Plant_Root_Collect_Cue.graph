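# Sound cue for collecting a health plant root ingredient, played as loot-pickup foley.
# A mixer plays three layers together: a root grab at reduced volume, a surreal plant collect sting and a bag-rustle layer.
# It routes through the loot collect sound class with prop-collect attenuation and a lowered overall volume.
# A concurrency override allows two simultaneous instances and stops the oldest.
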SoundCue /Game/Sounds/Foley_Loot/Ingredients/Health/Plant_Root_Collect_Cue
root=Mixer_1
nodes [
 {id=Mixer_1, children [WavePlayer_4,WavePlayer_3,WavePlayer_5], properties {inputVolume [0.7,1,1]}}
 {id=WavePlayer_3, properties {soundWaveAssetPtr=&/Plant_Collect_Surreal01}}
 {id=WavePlayer_4, properties {soundWaveAssetPtr=&/loot_root_get_01}}
 {id=WavePlayer_5, properties {soundWaveAssetPtr=/Game/Sounds/Foley_Loot/Ingredients/Bag/loot_setup_bagonly_02}}
]
properties {
 attenuationSettings=/Game/Sounds/Attenuation/Fort_Prop_Collect_Att
 bOverrideConcurrency=true
 maxDistance=2097152
 soundClassObject=&Fort_Loot_Collect
 volumeMultiplier=0.251
 concurrencyOverrides {maxCount=2, resolutionRule=StopOldest}
}
